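# Unlit test material for a cave natural-light glow whose color and brightness follow the time of day.
# Color_Time_of_Day blends warm morning, day and evening tints and a cyan night tint, then a constant boosts the result.
# Scalar_Time_of_Day scales that boosted color, peaking by day and dimmest at night.
# All values are hardcoded constants, with no textures or exposed parameters.
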Material /Game/Athena/Helios/Maps/Test/Dev/ChrisK/Materials/Helios_CaveNaturalLightGlow
graph {
 properties {pixelDepthOffsetMode=PDOM_Legacy, shadingModel=MSM_Unlit, usedShadingModels=Unlit}
 nodes [
  n1 = Constant3Vector() {constant=<rgba 1 0.5781 0.7186 1>}
  n2 = Constant3Vector() {constant=<rgba 1 0.9123 0.7292 1>}
  n3 = Constant3Vector() {constant=<rgba 1 0.5646 0.4479 1>}
  n4 = Constant3Vector() {constant=<rgba 0.25 0.8488 1 1>}
  n5 = Constant() {r=0.5}
  Multiply(
   A=Multiply(
    A=Constant() {r=3.5}
    B=MaterialFunctionCall(Color_Morning=@n1, Color_Day=@n2, Color_Evening=@n3, Color_Night=@n4) {materialFunction=&Color_Time_of_Day}.Result
   )
   B=MaterialFunctionCall(Value_Morning=@n5, Value_Day=Constant() {r=1}, Value_Evening=@n5, Value_Night=Constant() {r=0.2}) {materialFunction=&Scalar_Time_of_Day}.Result
  )
 ]
}
