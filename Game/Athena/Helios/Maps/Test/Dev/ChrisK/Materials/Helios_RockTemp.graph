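# Temporary rock material in a developer test map folder for Helios, apparently a flat-colour placeholder.
# It is a Default Lit material enabled for spline meshes and contains no textures.
# Its only nodes are three parameters, a dark neutral grey Color plus Roughness and Specular scalars, with high roughness as the default.
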Material /Game/Athena/Helios/Maps/Test/Dev/ChrisK/Materials/Helios_RockTemp
graph {
 properties {bUsedWithSplineMeshes=true, pixelDepthOffsetMode=PDOM_Legacy, usedShadingModels=Default Lit}
 nodes [
  ScalarParameter() {defaultValue=0.9, parameterName=Roughness}
  ScalarParameter() {defaultValue=0.5, parameterName=Specular}
  VectorParameter() {parameterName=Color, defaultValue=<rgba 0.1406 0.1406 0.1384 1>}
 ]
}
