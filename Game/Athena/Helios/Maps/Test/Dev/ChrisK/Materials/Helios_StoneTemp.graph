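# Helios_StoneTemp is a simple Default Lit material in a developer test map folder, apparently a temporary stand-in for stone.
# It exposes a dark grey Color vector parameter and scalar Roughness and Specular parameters, with Roughness declared twice at the same default.
# The graph contains no textures or math, and shows no links from the parameters to material outputs.
# It is flagged for use on spline meshes.
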Material /Game/Athena/Helios/Maps/Test/Dev/ChrisK/Materials/Helios_StoneTemp
graph {
 properties {bUsedWithSplineMeshes=true, pixelDepthOffsetMode=PDOM_Legacy, usedShadingModels=Default Lit}
 nodes [
  ScalarParameter() {defaultValue=0.5, parameterName=Roughness}
  ScalarParameter() {defaultValue=0.7, parameterName=Specular}
  ScalarParameter() {defaultValue=0.5, parameterName=Roughness}
  VectorParameter() {parameterName=Color, defaultValue=<rgba 0.0998 0.1002 0.1042 1>}
 ]
}
