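# Helios_TestTree is a simple Default Lit test material in a Helios developer map folder, named for use on a test tree.
# Its graph holds a Color vector parameter defaulting to dark green, plus two scalar constants.
# It contains no textures or math nodes, and the file does not show which material inputs the three nodes feed.
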Material /Game/Athena/Helios/Maps/Test/Dev/ChrisK/Materials/Helios_TestTree
graph {
 properties {pixelDepthOffsetMode=PDOM_Legacy, usedShadingModels=Default Lit}
 nodes [
  Constant() {r=0.75}
  Constant() {r=0.5}
  VectorParameter() {parameterName=Color, defaultValue=<rgba 0.0104 0.1042 0.0295 1>}
 ]
}
