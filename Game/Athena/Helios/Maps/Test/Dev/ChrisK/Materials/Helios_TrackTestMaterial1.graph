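# Simple test material in a Helios developer map folder, named for track testing and flagged for use on spline meshes.
# It uses the Default Lit shading model and exposes a dark grey Color vector parameter plus scalar Roughness and Specular parameters.
# The graph holds only these three parameter nodes, with no textures or math.
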Material /Game/Athena/Helios/Maps/Test/Dev/ChrisK/Materials/Helios_TrackTestMaterial1
graph {
 properties {bUsedWithSplineMeshes=true, pixelDepthOffsetMode=PDOM_Legacy, usedShadingModels=Default Lit}
 nodes [
  ScalarParameter() {defaultValue=0.3, parameterName=Roughness}
  ScalarParameter() {defaultValue=1, parameterName=Specular}
  VectorParameter() {parameterName=Color, defaultValue=<rgba 0.0521 0.0521 0.0521 0>}
 ]
}
