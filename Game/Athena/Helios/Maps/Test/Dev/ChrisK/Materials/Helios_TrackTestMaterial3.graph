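# Simple Default Lit test material in a Helios developer map folder, named for track testing and enabled for use on spline meshes.
# It exposes a brown Color vector parameter plus low Roughness and Specular scalars for material instances to override.
# The graph has no textures or math, only these three parameters, and shows no links between them and the material outputs.
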Material /Game/Athena/Helios/Maps/Test/Dev/ChrisK/Materials/Helios_TrackTestMaterial3
graph {
 properties {bUsedWithSplineMeshes=true, pixelDepthOffsetMode=PDOM_Legacy, usedShadingModels=Default Lit}
 nodes [
  ScalarParameter() {defaultValue=0.05, parameterName=Roughness}
  ScalarParameter() {defaultValue=0.01, parameterName=Specular}
  VectorParameter() {parameterName=Color, defaultValue=<rgba 0.1875 0.0848 0.0393 1>}
 ]
}
