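# Unlit test material for Helios track meshes, kept in a developer test map folder and enabled for use on spline meshes.
# Its graph holds a single Color vector parameter with a pink-red default, and no connection from it to a material output is shown.
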Material /Game/Athena/Helios/Maps/Test/Dev/ChrisK/Materials/Helios_TrackTestMaterial4
graph {
 properties {bUsedWithSplineMeshes=true, pixelDepthOffsetMode=PDOM_Legacy, shadingModel=MSM_Unlit, usedShadingModels=Unlit}
 nodes [VectorParameter() {parameterName=Color, defaultValue=<rgba 1 0 0.0636 1>}]
}
