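# Helios_WhiteTemp is a simple Default Lit material in a Helios developer test map folder, apparently a temporary plain surface for level work.
# It exposes a neutral grey Color vector parameter plus Roughness and Specular scalar parameters, with no textures or math nodes.
# Usage flags enable it on Nanite and spline meshes.
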Material /Game/Athena/Helios/Maps/Test/Dev/ChrisK/Materials/Helios_WhiteTemp
graph {
 properties {bUsedWithNanite=true, bUsedWithSplineMeshes=true, pixelDepthOffsetMode=PDOM_Legacy, usedShadingModels=Default Lit}
 nodes [
  ScalarParameter() {defaultValue=0.8, parameterName=Roughness}
  ScalarParameter() {defaultValue=0.6, parameterName=Specular}
  VectorParameter() {parameterName=Color, defaultValue=<rgba 0.2188 0.2188 0.2153 1>}
 ]
}
